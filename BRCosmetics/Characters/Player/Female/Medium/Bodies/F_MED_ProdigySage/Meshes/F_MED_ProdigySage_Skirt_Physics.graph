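# Physics asset for the F_MED_ProdigySage skirt, defining capsule bodies and joints for its dynamic bones.
# Eight chains (left/right A, B, C plus middle front and back) of three or four links hang from small MSTR anchor bodies jointed to the pelvis.
# Chain joints lock twist and limit swing, mostly loosening toward the hem.
# Larger capsules cover pelvis, spine and legs; their joints, like the anchors', are unrestricted in rotation.
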PhysicsAsset /BRCosmetics/Characters/Player/Female/Medium/Bodies/F_MED_ProdigySage/Meshes/F_MED_ProdigySage_Skirt_Physics
bodies {
 MSTR_SkirtA_L [{shapeType=Capsule, length=0.505, radius=0.505, shapeName=MSTR_SkirtA_L_capsule, rotation=<rot 0 -90 0>}]
 MSTR_SkirtA_R [{shapeType=Capsule, length=0.505, radius=0.505, shapeName=MSTR_SkirtA_R_capsule, rotation=<rot 0 -90 0>}]
 MSTR_SkirtB_L [{shapeType=Capsule, length=0.505, radius=0.505, shapeName=MSTR_SkirtB_L_capsule, rotation=<rot 0 -90 0>}]
 MSTR_SkirtB_R [{shapeType=Capsule, length=0.505, radius=0.505, shapeName=MSTR_SkirtB_R_capsule, rotation=<rot 0 -90 0>}]
 MSTR_SkirtC_L [{shapeType=Capsule, length=0.505, radius=0.505, shapeName=MSTR_SkirtC_L_capsule, rotation=<rot 0 -90 0>}]
 MSTR_SkirtC_R [{shapeType=Capsule, length=0.505, radius=0.505, shapeName=MSTR_SkirtC_R_capsule, rotation=<rot 0 -90 0>}]
 MSTR_SkirtM_Ba [{shapeType=Capsule, length=0.505, radius=0.505, shapeName=MSTR_SkirtM_Ba_capsule, rotation=<rot 0 -90 0>}]
 MSTR_SkirtM_Fr [{shapeType=Capsule, length=0.505, radius=0.505, shapeName=MSTR_SkirtM_Fr_capsule, rotation=<rot 0 -90 0>}]
 calf_l [{shapeType=Capsule, length=20, radius=10, shapeName=calf_l_capsule, center=<-14.132 0 0>, rotation=<rot -90 0 0>}]
 calf_r [{shapeType=Capsule, length=20, radius=10, shapeName=calf_r_capsule, center=<12.4216 0 0>, rotation=<rot -90 180 180>}]
 dyn_SkirtA_L_1 [{shapeType=Capsule, length=7, radius=3, shapeName=dyn_SkirtA_L_1_capsule, center=<13 -3.4 0>, rotation=<rot -90 0 0>}]
 dyn_SkirtA_L_2 [{shapeType=Capsule, length=7, radius=4, shapeName=dyn_SkirtA_L_2_capsule, center=<13 -0.1 0>, rotation=<rot 90 0 0>}]
 dyn_SkirtA_L_3 [{shapeType=Capsule, length=7, radius=4, shapeName=dyn_SkirtA_L_3_capsule, center=<13 1.7596 0>, rotation=<rot 90 0 0>}]
 dyn_SkirtA_R_1 [{shapeType=Capsule, length=7, radius=3, shapeName=dyn_SkirtA_R_1_capsule, center=<13 -3.4 0>, rotation=<rot -90 0 0>}]
 dyn_SkirtA_R_2 [{shapeType=Capsule, length=7, radius=3, shapeName=dyn_SkirtA_R_2_capsule, center=<13 -0.1 0>, rotation=<rot -90 180 180>}]
 dyn_SkirtA_R_3 [{shapeType=Capsule, length=7, radius=3, shapeName=dyn_SkirtA_R_3_capsule, center=<13 1.7596 0>, rotation=<rot -90 0 0>}]
 dyn_SkirtB_L_1 [{shapeType=Capsule, length=7, radius=3, shapeName=dyn_SkirtB_L_1_capsule, center=<13 -3.6 0>, rotation=<rot -90 0 0>}]
 dyn_SkirtB_L_2 [{shapeType=Capsule, length=7, radius=3, shapeName=dyn_SkirtB_L_2_capsule, center=<13 -1 0>, rotation=<rot 90 0 0>}]
 dyn_SkirtB_L_3 [{shapeType=Capsule, length=7, radius=3, shapeName=dyn_SkirtB_L_3_capsule, center=<13 1.7596 0>, rotation=<rot 90 0 0>}]
 dyn_SkirtB_R_1 [{shapeType=Capsule, length=7, radius=3, shapeName=dyn_SkirtB_R_1_capsule, center=<13 -3.6 0>, rotation=<rot -90 0 0>}]
 dyn_SkirtB_R_2 [{shapeType=Capsule, length=7, radius=3, shapeName=dyn_SkirtB_R_2_capsule, center=<13 -1 0>, rotation=<rot -90 0 0>}]
 dyn_SkirtB_R_3 [{shapeType=Capsule, length=7, radius=3, shapeName=dyn_SkirtB_R_3_capsule, center=<13 1.7596 0>, rotation=<rot -90 0 0>}]
 dyn_SkirtC_L_1 [{shapeType=Capsule, length=5, radius=3, shapeName=dyn_SkirtC_L_1_capsule, center=<8.5 -4.2 0>, rotation=<rot -90 0 0>}]
 dyn_SkirtC_L_2 [{shapeType=Capsule, length=5, radius=3, shapeName=dyn_SkirtC_L_2_capsule, center=<8.5 -3.6 0>, rotation=<rot -90 0 0>}]
 dyn_SkirtC_L_3 [{shapeType=Capsule, length=5, radius=3, shapeName=dyn_SkirtC_L_3_capsule, center=<14.2 -0.14 0>, rotation=<rot -90 0 0>}]
 dyn_SkirtC_L_4 [{shapeType=Capsule, length=5, radius=3, shapeName=dyn_SkirtC_L_4_capsule, center=<8.5 -0.7 0>, rotation=<rot -90 0 0>}]
 dyn_SkirtC_R_1 [{shapeType=Capsule, length=5, radius=3, shapeName=dyn_SkirtC_R_1_capsule, center=<8.5 -4.2 0>, rotation=<rot -90 180 180>}]
 dyn_SkirtC_R_2 [{shapeType=Capsule, length=5, radius=3, shapeName=dyn_SkirtC_R_2_capsule, center=<8.5 -2 0>, rotation=<rot -90 180 180>}]
 dyn_SkirtC_R_3 [{shapeType=Capsule, length=5, radius=3, shapeName=dyn_SkirtC_R_3_capsule, center=<14.3 0 0>, rotation=<rot -90 0 0>}]
 dyn_SkirtC_R_4 [{shapeType=Capsule, length=5, radius=3, shapeName=dyn_SkirtC_R_4_capsule, center=<8.5 -1.25 0>, rotation=<rot -90 180 180>}]
 dyn_SkirtM_Ba_1 [{shapeType=Capsule, length=5, radius=3, shapeName=dyn_SkirtM_Ba_1_capsule, center=<5 -2.4 0>, rotation=<rot -90 180 180>}]
 dyn_SkirtM_Ba_2 [{shapeType=Capsule, length=5, radius=3, shapeName=dyn_SkirtM_Ba_2_capsule, center=<5 -2 0>, rotation=<rot -90 180 180>}]
 dyn_SkirtM_Ba_3 [{shapeType=Capsule, length=5, radius=3, shapeName=dyn_SkirtM_Ba_3_capsule, center=<10.8 0 0>, rotation=<rot -90 0 0>}]
 dyn_SkirtM_Ba_4 [{shapeType=Capsule, length=5, radius=3, shapeName=dyn_SkirtM_Ba_4_capsule, center=<5 1.75 0>, rotation=<rot -90 0 0>}]
 dyn_SkirtM_Fr_1 [{shapeType=Capsule, length=7, radius=3, shapeName=dyn_SkirtM_Fr_1_capsule, center=<8 -3.2 0>, rotation=<rot -90 0 0>}]
 dyn_SkirtM_Fr_2 [{shapeType=Capsule, length=7, radius=3, shapeName=dyn_SkirtM_Fr_2_capsule, center=<8 0.6 0>}]
 dyn_SkirtM_Fr_3 [{shapeType=Capsule, length=7, radius=3, shapeName=dyn_SkirtM_Fr_3_capsule, center=<8 3.2 0>}]
 foot_l [{shapeType=Capsule, length=20, radius=10, shapeName=foot_l_capsule, center=<4.1353 0 0>, rotation=<rot -90 0 0>}]
 foot_r [{shapeType=Capsule, length=20, radius=10, shapeName=foot_r_capsule, center=<-7.5848 0 0>, rotation=<rot -90 180 180>}]
 pelvis [{shapeType=Capsule, length=9.5, radius=11, shapeName=pelvis_capsule, center=<-2.4713 -0.7357 0>}]
 root [{shapeType=Capsule, length=0.505, radius=0.505, shapeName=root_capsule, rotation=<rot 0 -90 0>}]
 spine_01 [{shapeType=Capsule, length=11.3926, radius=9.1468, shapeName=spine_01_capsule, center=<0 0.3843 0>}]
 spine_02 [{shapeType=Capsule, length=7.074, radius=7.787, shapeName=spine_02_capsule, center=<0 1.5228 0>}]
 spine_03 [{shapeType=Capsule, length=7.5409, radius=7.3202, shapeName=spine_03_capsule, center=<0 2.2426 0>}]
 thigh_l [{shapeType=Capsule, length=25.456, radius=8.5, shapeName=thigh_l_capsule, center=<-21.7424 -0.2986 0.5581>, rotation=<rot -90 180 180>}]
 thigh_r [{shapeType=Capsule, length=23.2944, radius=8.5, shapeName=thigh_r_capsule, center=<20 -0.1491 0>, rotation=<rot -90 0 0>}]
}
constraints [
 {
  bone1Name=pelvis
  bone2Name=root
  swing1LimitDegrees=45
  swing1Motion=Free
  swing2LimitDegrees=45
  swing2Motion=Free
  twistLimitDegrees=45
  twistMotion=Free
 }
 {
  bone1Name=MSTR_SkirtA_L
  bone2Name=pelvis
  swing1LimitDegrees=45
  swing1Motion=Free
  swing2LimitDegrees=45
  swing2Motion=Free
  twistLimitDegrees=45
  twistMotion=Free
 }
 {bone1Name=dyn_SkirtA_L_1, bone2Name=MSTR_SkirtA_L, swing1LimitDegrees=5, swing2LimitDegrees=10, twistLimitDegrees=0}
 {bone1Name=dyn_SkirtA_L_2, bone2Name=dyn_SkirtA_L_1, swing1LimitDegrees=15, swing2LimitDegrees=0, twistLimitDegrees=0}
 {bone1Name=dyn_SkirtA_L_3, bone2Name=dyn_SkirtA_L_2, swing1LimitDegrees=15, swing2LimitDegrees=0, twistLimitDegrees=0}
 {
  bone1Name=MSTR_SkirtA_R
  bone2Name=pelvis
  swing1LimitDegrees=45
  swing1Motion=Free
  swing2LimitDegrees=45
  swing2Motion=Free
  twistLimitDegrees=45
  twistMotion=Free
 }
 {bone1Name=dyn_SkirtA_R_1, bone2Name=MSTR_SkirtA_R, swing1LimitDegrees=5, swing2LimitDegrees=5, twistLimitDegrees=0}
 {bone1Name=dyn_SkirtA_R_2, bone2Name=dyn_SkirtA_R_1, swing1LimitDegrees=20, swing2LimitDegrees=0, twistLimitDegrees=0}
 {bone1Name=dyn_SkirtA_R_3, bone2Name=dyn_SkirtA_R_2, swing1LimitDegrees=20, swing2LimitDegrees=0, twistLimitDegrees=0}
 {
  bone1Name=MSTR_SkirtB_L
  bone2Name=pelvis
  swing1LimitDegrees=45
  swing1Motion=Free
  swing2LimitDegrees=45
  swing2Motion=Free
  twistLimitDegrees=45
  twistMotion=Free
 }
 {bone1Name=dyn_SkirtB_L_1, bone2Name=MSTR_SkirtB_L, swing1LimitDegrees=5, swing2LimitDegrees=5, twistLimitDegrees=0}
 {bone1Name=dyn_SkirtB_L_2, bone2Name=dyn_SkirtB_L_1, swing1LimitDegrees=30, swing2LimitDegrees=5, twistLimitDegrees=0}
 {bone1Name=dyn_SkirtB_L_3, bone2Name=dyn_SkirtB_L_2, swing1LimitDegrees=30, swing2LimitDegrees=5, twistLimitDegrees=0}
 {
  bone1Name=MSTR_SkirtB_R
  bone2Name=pelvis
  swing1LimitDegrees=45
  swing1Motion=Free
  swing2LimitDegrees=45
  swing2Motion=Free
  twistLimitDegrees=45
  twistMotion=Free
 }
 {bone1Name=dyn_SkirtB_R_1, bone2Name=MSTR_SkirtB_R, swing1LimitDegrees=5, swing2LimitDegrees=5, twistLimitDegrees=0}
 {bone1Name=dyn_SkirtB_R_2, bone2Name=dyn_SkirtB_R_1, swing1LimitDegrees=30, swing2LimitDegrees=5, twistLimitDegrees=0}
 {bone1Name=dyn_SkirtB_R_3, bone2Name=dyn_SkirtB_R_2, swing1LimitDegrees=30, swing2LimitDegrees=5, twistLimitDegrees=0}
 {
  bone1Name=MSTR_SkirtC_L
  bone2Name=pelvis
  swing1LimitDegrees=45
  swing1Motion=Free
  swing2LimitDegrees=45
  swing2Motion=Free
  twistLimitDegrees=45
  twistMotion=Free
 }
 {bone1Name=dyn_SkirtC_L_1, bone2Name=MSTR_SkirtC_L, swing1LimitDegrees=10, swing2LimitDegrees=3, twistLimitDegrees=0}
 {bone1Name=dyn_SkirtC_L_2, bone2Name=dyn_SkirtC_L_1, swing1LimitDegrees=20, swing2LimitDegrees=3, twistLimitDegrees=0}
 {bone1Name=dyn_SkirtC_L_3, bone2Name=dyn_SkirtC_L_2, swing1LimitDegrees=35, swing2LimitDegrees=3, twistLimitDegrees=0}
 {bone1Name=dyn_SkirtC_L_4, bone2Name=dyn_SkirtC_L_3, swing1LimitDegrees=55, swing2LimitDegrees=3, twistLimitDegrees=0}
 {
  bone1Name=MSTR_SkirtC_R
  bone2Name=pelvis
  swing1LimitDegrees=45
  swing1Motion=Free
  swing2LimitDegrees=45
  swing2Motion=Free
  twistLimitDegrees=45
  twistMotion=Free
 }
 {bone1Name=dyn_SkirtC_R_1, bone2Name=MSTR_SkirtC_R, swing1LimitDegrees=10, swing2LimitDegrees=3, twistLimitDegrees=0}
 {bone1Name=dyn_SkirtC_R_2, bone2Name=dyn_SkirtC_R_1, swing1LimitDegrees=20, swing2LimitDegrees=3, twistLimitDegrees=0}
 {bone1Name=dyn_SkirtC_R_3, bone2Name=dyn_SkirtC_R_2, swing1LimitDegrees=35, swing2LimitDegrees=3, twistLimitDegrees=0}
 {bone1Name=dyn_SkirtC_R_4, bone2Name=dyn_SkirtC_R_3, swing1LimitDegrees=55, swing2LimitDegrees=3, twistLimitDegrees=0}
 {
  bone1Name=MSTR_SkirtM_Ba
  bone2Name=pelvis
  swing1LimitDegrees=45
  swing1Motion=Free
  swing2LimitDegrees=45
  swing2Motion=Free
  twistLimitDegrees=45
  twistMotion=Free
 }
 {bone1Name=dyn_SkirtM_Ba_1, bone2Name=MSTR_SkirtM_Ba, swing1LimitDegrees=15, swing2LimitDegrees=3, twistLimitDegrees=0}
 {bone1Name=dyn_SkirtM_Ba_2, bone2Name=dyn_SkirtM_Ba_1, swing1LimitDegrees=15, swing2LimitDegrees=3, twistLimitDegrees=0}
 {bone1Name=dyn_SkirtM_Ba_3, bone2Name=dyn_SkirtM_Ba_2, swing1LimitDegrees=45, swing2LimitDegrees=10, twistLimitDegrees=0}
 {bone1Name=dyn_SkirtM_Ba_4, bone2Name=dyn_SkirtM_Ba_3, swing1LimitDegrees=45, swing2LimitDegrees=10, twistLimitDegrees=0}
 {
  bone1Name=MSTR_SkirtM_Fr
  bone2Name=pelvis
  swing1LimitDegrees=45
  swing1Motion=Free
  swing2LimitDegrees=45
  swing2Motion=Free
  twistLimitDegrees=45
  twistMotion=Free
 }
 {bone1Name=dyn_SkirtM_Fr_1, bone2Name=MSTR_SkirtM_Fr, swing1LimitDegrees=5, swing2LimitDegrees=5, twistLimitDegrees=0}
 {bone1Name=dyn_SkirtM_Fr_2, bone2Name=dyn_SkirtM_Fr_1, swing1LimitDegrees=20, swing2LimitDegrees=0, twistLimitDegrees=0}
 {bone1Name=dyn_SkirtM_Fr_3, bone2Name=dyn_SkirtM_Fr_2, swing1LimitDegrees=20, swing2LimitDegrees=0, twistLimitDegrees=0}
 {
  bone1Name=spine_01
  bone2Name=pelvis
  swing1LimitDegrees=45
  swing1Motion=Free
  swing2LimitDegrees=45
  swing2Motion=Free
  twistLimitDegrees=45
  twistMotion=Free
 }
 {
  bone1Name=spine_02
  bone2Name=spine_01
  swing1LimitDegrees=45
  swing1Motion=Free
  swing2LimitDegrees=45
  swing2Motion=Free
  twistLimitDegrees=45
  twistMotion=Free
 }
 {
  bone1Name=spine_03
  bone2Name=spine_02
  swing1LimitDegrees=45
  swing1Motion=Free
  swing2LimitDegrees=45
  swing2Motion=Free
  twistLimitDegrees=45
  twistMotion=Free
 }
 {
  bone1Name=thigh_l
  bone2Name=pelvis
  swing1LimitDegrees=45
  swing1Motion=Free
  swing2LimitDegrees=45
  swing2Motion=Free
  twistLimitDegrees=45
  twistMotion=Free
 }
 {
  bone1Name=calf_l
  bone2Name=thigh_l
  swing1LimitDegrees=45
  swing1Motion=Free
  swing2LimitDegrees=45
  swing2Motion=Free
  twistLimitDegrees=45
  twistMotion=Free
 }
 {
  bone1Name=foot_l
  bone2Name=calf_l
  swing1LimitDegrees=45
  swing1Motion=Free
  swing2LimitDegrees=45
  swing2Motion=Free
  twistLimitDegrees=45
  twistMotion=Free
 }
 {
  bone1Name=thigh_r
  bone2Name=pelvis
  swing1LimitDegrees=45
  swing1Motion=Free
  swing2LimitDegrees=45
  swing2Motion=Free
  twistLimitDegrees=45
  twistMotion=Free
 }
 {
  bone1Name=calf_r
  bone2Name=thigh_r
  swing1LimitDegrees=45
  swing1Motion=Free
  swing2LimitDegrees=45
  swing2Motion=Free
  twistLimitDegrees=45
  twistMotion=Free
 }
 {
  bone1Name=foot_r
  bone2Name=calf_r
  swing1LimitDegrees=45
  swing1Motion=Free
  swing2LimitDegrees=45
  swing2Motion=Free
  twistLimitDegrees=45
  twistMotion=Free
 }
]
