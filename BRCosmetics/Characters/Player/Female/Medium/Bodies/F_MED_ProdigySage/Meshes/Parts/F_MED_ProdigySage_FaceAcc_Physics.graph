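# Physics asset for the F_MED_ProdigySage face accessory mesh, providing physics bodies for a left and a right hair strand.
# Each strand chains three larger dyn_ capsules from a tiny MSTR_ root capsule, with constraints linking each bone to the one before it.
# Twist is locked on every joint.
# The sides are asymmetric: the two lower right joints also lock Swing2, while the left joints do not.
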PhysicsAsset /BRCosmetics/Characters/Player/Female/Medium/Bodies/F_MED_ProdigySage/Meshes/Parts/F_MED_ProdigySage_FaceAcc_Physics
bodies {
 MSTR_hairStrandB_L [{shapeType=Capsule, length=0.505, radius=0.505, shapeName=MSTR_hairStrandB_L_capsule}]
 MSTR_hairStrandB_R [{shapeType=Capsule, length=0.505, radius=0.505, shapeName=MSTR_hairStrandB_R_capsule}]
 dyn_hairStrandB_L_01 [{shapeType=Capsule, length=6, radius=4, shapeName=dyn_hairStrandB_L_01_capsule, center=<6.8368 0 0>, rotation=<rot -90 0 0>}]
 dyn_hairStrandB_L_02 [{shapeType=Capsule, length=4.7406, radius=3.8466, shapeName=dyn_hairStrandB_L_02_capsule, center=<2.2896 0 0>, rotation=<rot -90 0 0>}]
 dyn_hairStrandB_L_03 [{shapeType=Capsule, length=6, radius=4, shapeName=dyn_hairStrandB_L_03_capsule, center=<2.2896 0 0>, rotation=<rot -90 0 0>}]
 dyn_hairStrandB_R_01 [{shapeType=Capsule, length=6, radius=4, shapeName=dyn_hairStrandB_R_01_capsule, center=<6.6795 0 0>, rotation=<rot -90 180 180>}]
 dyn_hairStrandB_R_02 [{shapeType=Capsule, length=6, radius=4, shapeName=dyn_hairStrandB_R_02_capsule, center=<2.2896 0 0>, rotation=<rot -90 0 0>}]
 dyn_hairStrandB_R_03 [{shapeType=Capsule, length=6, radius=4, shapeName=dyn_hairStrandB_R_03_capsule, center=<2.2896 0 0>, rotation=<rot -90 0 0>}]
}
constraints [
 {
  bone1Name=dyn_hairStrandB_R_01
  bone2Name=MSTR_hairStrandB_R
  swing1LimitDegrees=5
  swing2LimitDegrees=5
  twistLimitDegrees=45
  twistMotion=Locked
 }
 {
  bone1Name=dyn_hairStrandB_R_02
  bone2Name=dyn_hairStrandB_R_01
  swing1LimitDegrees=5
  swing2LimitDegrees=5
  swing2Motion=Locked
  twistLimitDegrees=45
  twistMotion=Locked
 }
 {
  bone1Name=dyn_hairStrandB_R_03
  bone2Name=dyn_hairStrandB_R_02
  swing1LimitDegrees=5
  swing2LimitDegrees=5
  swing2Motion=Locked
  twistLimitDegrees=45
  twistMotion=Locked
 }
 {
  bone1Name=dyn_hairStrandB_L_01
  bone2Name=MSTR_hairStrandB_L
  swing1LimitDegrees=5
  swing2LimitDegrees=5
  twistLimitDegrees=45
  twistMotion=Locked
 }
 {
  bone1Name=dyn_hairStrandB_L_02
  bone2Name=dyn_hairStrandB_L_01
  swing1LimitDegrees=5
  swing2LimitDegrees=5
  twistLimitDegrees=45
  twistMotion=Locked
 }
 {
  bone1Name=dyn_hairStrandB_L_03
  bone2Name=dyn_hairStrandB_L_02
  swing1LimitDegrees=5
  swing2LimitDegrees=5
  twistLimitDegrees=45
  twistMotion=Locked
 }
]
